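# Physics asset for the Cute Duo Male pickaxe, simulating a balloon tethered to the two-handed weapon handle.
# The handle body uses four boxes clustered around the upper handle as a collision volume.
# Three small capsules form a line chain from handle to a larger capsule on the balloon bone.
# Angular limits are tighter at both chain ends and looser between line segments, letting the balloon sway within bounds.
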PhysicsAsset /BRCosmetics/Weapons/FORT_Melee/Pickaxe_Cute_Duo_Male/Meshes/Cute_Duo_Male_Phy_balloon
bodies {
 Melee_TwoHanded_Handle [
  {shapeType=Box, extentX=7.871, extentY=40, extentZ=4, center=<0 -20.5983 72.99>, rotation=<rot 0 0 89.9994>}
  {shapeType=Box, extentX=7.871, extentY=40, extentZ=4, center=<0 20.4435 72.9699>, rotation=<rot 0 0 89.9997>}
  {shapeType=Box, extentX=15.0001, extentY=10.0007, extentZ=4, center=<2.1381 -0.7723 51.9581>, rotation=<rot 0.0007 89.9996 89.9999>}
  {shapeType=Box, extentX=7.871, extentY=40, extentZ=4, center=<0 -0.3661 91.2997>, rotation=<rot 0 0 179.9991>}
 ]
 dyn_balloon [{shapeType=Capsule, length=10.0971, radius=11.0418, center=<13.0766 0.0105 -0.2411>, rotation=<rot 0 0.2498 -89.9996>}]
 dyn_line_01 [{shapeType=Capsule, length=2, radius=1, center=<1 0 0>, rotation=<rot 90 0 0>}]
 dyn_line_02 [{shapeType=Capsule, length=2, radius=1, center=<1 0 0>, rotation=<rot 90 0 0>}]
 dyn_line_03 [{shapeType=Capsule, length=2, radius=1, center=<1 0 0>, rotation=<rot 90 0 0>}]
}
constraints [
 {bone1Name=dyn_line_01, bone2Name=Melee_TwoHanded_Handle, swing1LimitDegrees=15, swing2LimitDegrees=30, twistLimitDegrees=5}
 {bone1Name=dyn_line_02, bone2Name=dyn_line_01, swing1LimitDegrees=30, swing2LimitDegrees=45, twistLimitDegrees=10}
 {bone1Name=dyn_line_03, bone2Name=dyn_line_02, swing1LimitDegrees=30, swing2LimitDegrees=45, twistLimitDegrees=10}
 {bone1Name=dyn_balloon, bone2Name=dyn_line_03, swing1LimitDegrees=15, swing2LimitDegrees=30, twistLimitDegrees=5}
]
