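# Physics asset for the strap on the Cute Duo Male two-handed pickaxe cosmetic, letting the strap dangle and swing.
# One long capsule body covers the weapon handle bone and a short capsule covers the dyn_strap bone.
# A single constraint joins the strap to the handle with both swing axes free and a narrow twist limit.
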PhysicsAsset /BRCosmetics/Weapons/FORT_Melee/Pickaxe_Cute_Duo_Male/Meshes/Cute_Duo_Male_Phy_strap
bodies {
 Melee_TwoHanded_Handle [{shapeType=Capsule, length=99.411, radius=1.6195, center=<0.9804 -1.2604 29.4546>, rotation=<rot 0.2898 -90 0.2004>}]
 dyn_strap [{shapeType=Capsule, length=7.5, radius=1, center=<4.5 0 0>, rotation=<rot 90 0 0>}]
}
constraints [{
  bone1Name=dyn_strap
  bone2Name=Melee_TwoHanded_Handle
  swing1LimitDegrees=90
  swing1Motion=Free
  swing2LimitDegrees=90
  swing2Motion=Free
  twistLimitDegrees=5
 }]
