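# Animation Blueprint for the Cute Duo Male pickaxe, adding secondary motion to its balloon and strap.
# Two chained RigidBody nodes simulate in handle-bone space: the balloon with slight upward gravity for float, then the strap with downward gravity.
# A bool blend on bRigidBodyEnabled picks that result, or a lighter rotation-limited Trail controller on dyn_line_03 when false, both fed from the cached input pose.
# An unconnected balloon Spring controller remains, and the EventGraph has no logic.
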AnimBlueprint /BRCosmetics/Weapons/FORT_Melee/Pickaxe_Cute_Duo_Male/Meshes/Cute_Duo_Male_AnimBP
target_skeleton=&/Cute_Duo_Male_Skeleton
graphs [
 AnimGraph {nodes [
   n1 = RigidBody(ComponentPose=Local To Component(LocalPose=Use cached pose 'PrePose'())) {
    bClampLinearTranslationLimitToRefPose=true
    bOverrideWorldGravity=true
    overridePhysicsAsset=&/Cute_Duo_Male_Phy_balloon
    simulationSpace=BaseBoneSpace
    alphaBoolBlend {blendOption=Linear}
    baseBoneRef {boneName=Melee_TwoHanded_Handle}
    componentAppliedLinearAccClamp=<500 500 500>
    componentLinearAccScale=<0.15 0.15 0.15>
    componentLinearVelScale=<0.1 0.1 0.1>
    overrideWorldGravity {z=10}
   }
   n2 = RigidBody(ComponentPose=@n1) {
    bClampLinearTranslationLimitToRefPose=true
    bOverrideWorldGravity=true
    overridePhysicsAsset=&/Cute_Duo_Male_Phy_strap
    simulationSpace=BaseBoneSpace
    alphaBoolBlend {blendOption=Linear}
    baseBoneRef {boneName=Melee_TwoHanded_Handle}
    componentAppliedLinearAccClamp=<3000 3000 3000>
    componentLinearAccScale=<0.3 0.3 0.4>
    componentLinearVelScale=<0.2 0.2 0.2>
    overrideWorldGravity {z=-600}
   }
   n3 = Trail controller - Bone: dyn_line_03(ComponentPose=Local To Component(LocalPose=Use cached pose 'PrePose'()), Alpha=1) {
    bLimitRotation=true
    stretchLimit=0.5
    rotationLimits [{limitMax=<0 180 0>, limitMin=<0 -180 0>}, {limitMax=<0 180 0>, limitMin=<0 -180 0>}]
    rotationOffsets [<0 0 0>, <0 0 0>]
    trailBone {boneName=dyn_line_03}
   }
   Output Pose(Result=Blend Poses by bool(BlendPose_0=Component To Local(ComponentPose=@n2), BlendPose_1=Component To Local(ComponentPose=@n3), bActiveValue=Get bRigidBodyEnabled()) {blendType=Linear}) {layerGroup=None}
   Save cached pose 'PrePose'(Pose=Input Pose()) {cacheName=PrePose}
   Spring controller - Bone: dyn_balloon!(ComponentPose:Component Space Pose Link Structure, Alpha:float=1) -> (Pose:Component Space Pose Link Structure) {
    is=title AnimGraphNode_SpringBone
    bRotateX=true
    bRotateY=true
    bRotateZ=true
    bTranslateX=false
    bTranslateY=false
    bTranslateZ=false
    springDamping=1
    springStiffness=10
    alphaBoolBlend {blendOption=Linear}
    springBone {boneName=dyn_balloon}
   }
  ]}
 EventGraph {nodes [
   TryGetPawnOwner()
   Event BlueprintUpdateAnimation()
  ]}
]
variables []
